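; Farm respawn chunk Blueprint, apparently one of the abducted POI environment chunks, that sizes the Niagara effect for its underside to match its static mesh.
; Both the construction script and BeginPlay read the mesh component's bounds and pass the sphere radius to the "SpawnSphere" float parameter on NS_Mother_POIChunk_Under.
; BeginPlay also calls the parent implementation, while the overlap and tick events are empty.
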(event EventBeginPlay
  (bind (_origin _boxextent _sphereradius) (Collision|GetComponentBounds (Variables|Default|GetStaticMesh)))
  (|CallParentFunction)
  (Effects|Components|ParticleSystem|SetFloatParameter (Variables|Default|GetNS_Mother_POIChunk_Under) "SpawnSphere" _sphereradius))

(event Collision|EventActorBeginOverlap
  (params (OtherActor :type Actor%20Object%20Reference)))

(event EventTick
  (params (DeltaSeconds :type Float%20%28single-precision%29)))

(fn ConstructionScript
  (bind (_origin _boxextent _sphereradius) (Collision|GetComponentBounds (Variables|Default|GetStaticMesh)))
  (Effects|Components|ParticleSystem|SetFloatParameter (Variables|Default|GetNS_Mother_POIChunk_Under) "SpawnSphere" _sphereradius))
