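; Farm shack chunk actor from the abducted-POI set, a static mesh actor with an under-chunk Niagara effect (NS_Mother_POIChunk_Under).
; It reads the static mesh component's bounds and feeds the sphere radius into the effect's SpawnSphere float parameter, so the effect scales with the mesh.
; This runs in both the construction script and BeginPlay, where BeginPlay calls the parent first.
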(event EventBeginPlay
  (bind (_origin _boxextent _sphereradius) (Collision|GetComponentBounds (Variables|StaticMeshActor|GetStaticMeshComponent)))
  (|CallParentFunction)
  (Effects|Components|ParticleSystem|SetFloatParameter (Variables|Default|GetNS_Mother_POIChunk_Under) "SpawnSphere" _sphereradius))

(fn ConstructionScript
  (bind (_origin _boxextent _sphereradius) (Collision|GetComponentBounds (Variables|StaticMeshActor|GetStaticMeshComponent)))
  (Effects|Components|ParticleSystem|SetFloatParameter (Variables|Default|GetNS_Mother_POIChunk_Under) "SpawnSphere" _sphereradius))
